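# Niagara pickup effect for monster parts, apparently a persistent magenta glow marker with a world-oriented mesh.
# The Main emitter burst-spawns particles that outlive their lifetime, feeding two sprite renderers, a light renderer and a mesh renderer.
# Dark-glow, beam and dynamic material parameters are set in particle update, with time-of-day selectors scaling alpha and light attributes following particle color.
# The Rays emitter has no renderers.
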NiagaraSystem /CorruptionGameplay/FX/NS_Pickups_MonsterParts
dependencies {usedDataInterfaces [], usedDynamicInputs [], usedModules [&SystemState]}
emitters {
 Main {
  emitterUpdateScript [
   &EmitterState(Life Cycle Mode=::NewEnumerator1, Loop Behavior=::NewEnumerator1, Loop Duration Mode=::NewEnumerator1, UseLoopDelay=false)
   &SpawnBurst_Instantaneous(Use Spawn Probability=false, UseLoopCountLimit=false)
   /SetVariables(Emitter.SpeedY=-0.25) {isSetParameters=true}
  ]
  particleSpawnScript [
   &InitializeParticle(
    Color=<rgba 0.5 0 0.3333 1>
    Mesh Scale=<0.2 1 1>
    Mesh Scale Mode=::NewEnumerator4
    Sprite Size Mode=::NewEnumerator3
    Uniform Sprite Size=150
    UsePositionOffset=true
    Write Material Random=true
   )
   &CameraOffset(Camera Offset Amount=-5)
   &InitialMeshOrientation(Mesh Orientation Mode=::NewEnumerator2, Orientation Coordinate Space=::World, Orientation Vector=<0 0 1>, Use Rotation Vector=true)
  ]
  particleUpdateScript [
   /SetVariables(
    Particles.DarkGlowCamOffset=-5
    Particles.DarkGlowColor:LinearColor=&MultiplyLinearColor(
     Linear Color=<rgba 5 0 3.3333 0.15>
     Scale Factor=&MakeVector4FromVectorAndFloat(
      "Vector (RGB)"=<1 1 1>
      "Float (Alpha)"=&FN_Select_Float_from_Time_Of_Day(Morning=0.8, Day=1, Evening=0.8, Night=0.6)
     )
    )
    Particles.DarkGlowDynPar:Vector4f=<3 0 0 0>
    Particles.DarkGlowSize:Vector2f=&Vector2DFromFloat(Value=60)
   ) {isSetParameters=true}
   /SetVariables(Particles.DymBeam:Vector4f=<30 0 0 2>) {isSetParameters=true}
   &ParticleState(Kill Particles When Lifetime Has Elapsed=false, Let Infinitely Lived Particles Die When Emitter Deactivates=true)
   &ScaleColor(
    Scale Alpha=&FN_Select_Float_from_Time_Of_Day(Morning=0.7, Day=1, Evening=0.7, Night=0.5)
    ScaleA=true
    ScaleRGB=true
   )
   &DynamicMaterialParameters(
    Index 0 Param 1=1
    Index 0 Param 2=5
    Module.Index0Param0Write=true
    Module.Index0Param1Write=true
    Module.Index0Param2Write=false
    Module.Index0Param3Write=false
   )
   &FN_Light_Attributes(
    Color=$Particles.Color
    Create Color Attribute=true
    Create Exponent Attribute=true
    Create Radius Attribute=true
    Exponent=15
    Radius=30
    Scale Color=10
   )
  ]
  rendererClasses [
   &NiagaraSpriteRendererProperties
   &NiagaraLightRendererProperties
   &NiagaraMeshRendererProperties
  ]
  renderers [
   {rendererClass=&NiagaraSpriteRendererProperties}
   {rendererClass=&NiagaraLightRendererProperties, rendererIndex=1}
   {rendererClass=&NiagaraSpriteRendererProperties, rendererIndex=2}
   {rendererClass=&NiagaraMeshRendererProperties, rendererIndex=3}
  ]
 }
 Rays {rendererClasses [], renderers []}
}
summary {
 systemName=NS_Pickups_MonsterParts
 emitters [
  {
   emitterName=Main
   rendererClasses [
    &NiagaraSpriteRendererProperties
    &NiagaraLightRendererProperties
    &NiagaraMeshRendererProperties
   ]
  }
  {emitterName=Rays, rendererClasses []}
 ]
 userVariables []
}
