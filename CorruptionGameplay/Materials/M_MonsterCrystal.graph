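# M_MonsterCrystal is a lit master material for a monster crystal artifact with animated flame-like glow.
# A static switch enables vertex rotation around the local Z axis, with corrected normals and a world-position-phased sine bob.
# Base attributes and normal map pass through MF_Crystal for the crystal look.
# Emissive adds view-angle-shaped flames from scrolling, dithered-triplanar Perlin noise, distorted by normal and base color and masked by base color red.
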Material /CorruptionGameplay/Materials/M_MonsterCrystal
graph {
 properties {bAutomaticallySetUsageInEditor=false, bUseMaterialAttributes=true, shadingModel=MSM_FromMaterialExpression, usedShadingModels=Default Lit}
 nodes [
  Constant3Vector()
  LocalPosition()
  n1 = VertexInterpolator(VS=LocalPosition() {includedOffsets=ExcludeOffsets, localOrigin=InstancePreSkinning}.XYZ)
  n2 = Multiply(A=ObjectRadius(), B=ScalarParameter() {defaultValue=1, group=ToadjamArtifact, parameterName=Param_1})
  MaterialFunctionCall(
   In=Distance(
    A=MaterialFunctionCall() {materialFunction=&ComponentPivotLocation}.Result
    B=Divide(A=@n1.PS, B=@n2) {constB=100}
   )
   Contrast=ScalarParameter() {defaultValue=0.5, group=ToadjamArtifact, parameterName=Param_4}
  ) {materialFunction=&CheapContrast}
  n3 = ScalarParameter() {defaultValue=0.1, group=ToadjamArtifact, parameterName=MeshRotationSpeed}
  n4 = Constant3Vector() {constant=<rgba 1 0 0 0>}
  n5 = Time()
  n6 = GetMaterialAttributes(None=MakeMaterialAttributes!(
    BaseColor=VectorParameter() {group=ToadjamArtifact, parameterName=Color, defaultValue=<rgba 0.3438 0.0537 0.1054 1>}.RGB
    Metallic=ScalarParameter() {group=ToadjamArtifact, parameterName=Metallic}
    Specular=ScalarParameter() {defaultValue=0.2, group=ToadjamArtifact, parameterName=Specular}
    Roughness=ScalarParameter() {defaultValue=0.5, group=ToadjamArtifact, parameterName=Roughness}
    Anisotropy
    EmissiveColor
    Opacity
    OpacityMask
    Normal=TextureSample() {samplerType=SAMPLERTYPE_Normal, texture=/CorruptionGameplay/Textures/T_MonsterCrystal_N}.RGB
    Tangent
    WorldPositionOffset
    SubsurfaceColor
    ClearCoat
    ClearCoatRoughness
    AmbientOcclusion
    Refraction
    CustomizedUVs_0
    CustomizedUVs_1
    CustomizedUVs_2
    CustomizedUVs_3
    CustomizedUVs_4
    CustomizedUVs_5
    CustomizedUVs_6
    CustomizedUVs_7
    PixelDepthOffset
    ShadingModel
    Displacement
   )) {attributeGetTypes [#1]}
  n7 = MaterialFunctionCall!(
   RotationAngle=Multiply(A=Multiply(A=@n3, B=@n4), B=NamedRerouteDeclaration(None=@n5) {name=Time_IN, nodeColor=<rgba 0.8281 0 1 1>})
   NormalizedRotationAxis=Transform(None=Constant3Vector() {constant=<rgba 0 0 1 0>}) {transformSourceType=TRANSFORMSOURCE_Local}
   PivotPoint=TransformPosition(None=Constant3Vector()) {transformType=TRANSFORMPOSSOURCE_World}
   Position=WorldPosition().XYZ
   Normal=Transform(None=@n6.Normal)
   TangentU
   TangentV
  ) {materialFunction=/Game/Characters/Player/Male/Medium/Bodies/M_MED_Inferno/FX/Mat/RotateVertexAboutAxis}
  n8 = Transform(None=@n7.Normal) {transformSourceType=TRANSFORMSOURCE_World, transformType=TRANSFORM_Tangent}
  n9 = NamedRerouteDeclaration(None=Divide(A=WorldPosition() {worldPositionShaderOffset=WPT_ExcludeAllShaderOffsets}.XYZ) {constB=128}) {name=WSRandomOffset, nodeColor=<rgba 1 0.4922 0 1>}
  n10 = ComponentMask(None=@n9) {r=true}
  n11 = ComponentMask(None=@n9) {g=true}
  n12 = Add(A=NamedRerouteUsage(), B=Add(A=Add(A=@n10, B=@n11), B=ComponentMask(None=@n9) {b=true}))
  n13 = Add(A=@n7.WPO, B=Multiply(A=Sine(None=@n12) {period=3.3}, B=Constant3Vector() {constant=<rgba 0 0 3 0>}))
  n14 = MaterialFunctionCall!(
   Material=StaticSwitchParameter(
    True=SetMaterialAttributes!(MaterialAttributes=@n6.MaterialAttributes, Normal=@n8, World Position Offset=@n13) {attributeSetTypes [#1, #2]}
    False=@n6.MaterialAttributes
   ) {defaultValue=true, group=ToadjamArtifact, parameterName=Rotate Around Axis}
   Alpha=Constant() {r=1}
  ) {materialFunction=/Game/Global/MaterialFunctions/TechArt/MF_Crystal}
  n15 = GetMaterialAttributes(None=@n14.Result) {attributeGetTypes [#3, #4, #1]}
  n16 = ComponentMask(None=@n15.Base Color) {r=true}
  n17 = Constant() {r=0.3}
  n18 = GetMaterialAttributes(None=@n14.Result) {attributeGetTypes [#3, #4, #1]}
  n19 = Transform(None=@n18.Normal)
  n20 = Add(A=DotProduct(A=@n19, B=CameraVectorWS()), B=ScalarParameter() {defaultValue=-0.15, group=ToadjamArtifact, parameterName=Flames_Frenel Bias}) {constB=-0.5}
  n21 = ScalarParameter() {defaultValue=1.5, group=ToadjamArtifact, parameterName=Flames_FresnelContrast}
  n22 = MaterialFunctionCall!(
   "Normal (WS)"
   Texture Scale=Divide(A=ObjectRadius(), B=ScalarParameter() {defaultValue=0.5, group=ToadjamArtifact, parameterName=Flames_Tile})
   Triplanar Fade=Constant() {r=10}
   Derivatives Scale=Constant() {r=0.5}
  ) {materialFunction=/Game/Athena/Environments/MaterialFunctions/MF_TriplanarDithered}
  n23 = Multiply(A=Time(), B=ScalarParameter() {defaultValue=0.2, group=ToadjamArtifact, parameterName=Flames_Scroll Speed}) {constB=0.2}
  n24 = Add(A=ComponentMask(None=@n19) {g=true, r=true}, B=ComponentMask(None=@n18.Base Color) {g=true, r=true})
  n25 = Constant() {r=0.3}
  n26 = TextureSample!(UVs=Subtract(A=Add(A=@n22.UVs, B=@n23), B=Multiply(A=@n24, B=@n25)), Tex, "DDX(UVs)"=@n22.DDX, "DDY(UVs)"=@n22.DDX, Apply View MipBias) {
   mipValueMode=TMVM_Derivative
   samplerType=SAMPLERTYPE_LinearGrayscale
   texture=&T_SimplePerlin_Greyscale
  }
  n27 = MaterialFunctionCall(
   In=Subtract(A=Multiply(A=@n20, B=@n21) {constB=2}, B=@n26.RGB)
   Contrast=ScalarParameter() {defaultValue=0.2, group=ToadjamArtifact, parameterName=Flames_Cheap Contrast}
  ) {materialFunction=&CheapContrast}
  n28 = ConstantBiasScale(None=@n27.Result) {bias=-0.75, scale=3}
  n29 = Multiply(
   A=MaterialFunctionCall(In=Add(A=@n16, B=@n17), Contrast=Constant() {r=1.5}) {materialFunction=&CheapContrast}.Result
   B=Add(A=Saturate(None=@n28), B=@n27.Result)
  )
  n30 = VectorParameter() {group=ToadjamArtifact, parameterName=Flames_EmissiveColor, defaultValue=<rgba 1 0 0.1965 1>}
  SetMaterialAttributes!(MaterialAttributes=@n15.MaterialAttributes, Emissive Color=Add(A=@n15.Emissive Color, B=Multiply(A=@n29, B=@n30.RGB))) {attributeSetTypes [#3]}
  TransformPosition(None=MaterialFunctionCall() {materialFunction=&ObjectPivotPoint}.Object Pivot Location) {transformType=TRANSFORMPOSSOURCE_World}
  Transform(None=VertexNormalWS()) {transformSourceType=TRANSFORMSOURCE_World, transformType=TRANSFORM_Tangent}
  VertexColor()
  WorldPosition()
 ]
}
