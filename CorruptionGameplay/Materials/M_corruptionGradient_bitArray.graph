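# Unlit material in CorruptionGameplay, named as a corruption gradient bit array; how it produces that is not evident in the graph.
# A texture object parameter named Corruption defaults to the RT_CorruptionProcessed render target from the Apollo landscape corruption materials and feeds a mask-type texture sample.
# That sample and a separate constant show no connection to any material output, so the graph appears to be a stub.
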Material /CorruptionGameplay/Materials/M_corruptionGradient_bitArray
graph {
 properties {pixelDepthOffsetMode=PDOM_Legacy, shadingModel=MSM_Unlit, usedShadingModels=Unlit}
 nodes [
  Constant() {r=1}
  n1 = TextureObjectParameter() {
   parameterName=Corruption
   samplerType=SAMPLERTYPE_LinearColor
   texture=/Game/Athena/Apollo/Environments/Landscape/Materials/Corruption/RT_CorruptionProcessed
  }
  TextureSample(Tex=@n1) {samplerType=SAMPLERTYPE_Masks, texture=/Game/Global/Textures/Default/Noise/NoisePattern}
 ]
}
