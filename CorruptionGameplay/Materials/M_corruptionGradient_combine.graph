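# Unlit material that merges a footstep mask into a layered corruption gradient texture, packing time and coverage into channels.
# Where footstep alpha is at or above a small threshold, X takes the CorruptionTime parameter, otherwise it keeps the layered texture's red value.
# Y is the saturated sum of footstep alpha and layered green, so coverage accumulates, and MakeFloat3 packs both values.
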Material /CorruptionGameplay/Materials/M_corruptionGradient_combine
graph {
 properties {pixelDepthOffsetMode=PDOM_Legacy, shadingModel=MSM_Unlit, usedShadingModels=Unlit}
 nodes [
  n1 = TextureSample(Tex=TextureObjectParameter() {parameterName=FootstepTexture, samplerType=SAMPLERTYPE_Masks, texture=&T_BlackMask}) {samplerType=SAMPLERTYPE_Masks}
  n2 = ScalarParameter() {parameterName=CorruptionTime}
  n3 = TextureSample(Tex=TextureObjectParameter() {parameterName=LayeredTexture, samplerType=SAMPLERTYPE_Masks, texture=&T_BlackMask}) {samplerType=SAMPLERTYPE_Masks}
  MaterialFunctionCall(X=If(A=@n1.A, B=Constant() {r=0.02}, "A > B"=@n2, "A == B"=@n2, "A < B"=@n3.R), Y=Saturate(None=Add(A=@n1.A, B=@n3.G))) {materialFunction=&MakeFloat3}
 ]
}
