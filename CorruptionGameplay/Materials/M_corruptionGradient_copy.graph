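# M_corruptionGradient_copy is a small unlit material in CorruptionGameplay, named as a copy of a corruption gradient material.
# It samples a texture object parameter named Texture, defaulting to T_BlackMask, as a mask.
# The red channel feeds both X and Y of MakeFloat3, with Z left unwired.
# The graph shows no connection from that vector to a material output.
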Material /CorruptionGameplay/Materials/M_corruptionGradient_copy
graph {
 properties {pixelDepthOffsetMode=PDOM_Legacy, shadingModel=MSM_Unlit, usedShadingModels=Unlit}
 nodes [
  n1 = TextureSample(Tex=TextureObjectParameter() {parameterName=Texture, samplerType=SAMPLERTYPE_Masks, texture=&T_BlackMask}) {samplerType=SAMPLERTYPE_Masks}
  MaterialFunctionCall(X=@n1.R, Y=@n1.R) {materialFunction=&MakeFloat3}
 ]
}
