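# Unlit processing material that builds a corruption gradient mask from a thresholded texture and initial corruption areas.
# Adds the thresholded red channel to two spiral blurs, one noise-modulated, then maxes it with the starting-area red channel and another mask.
# A second-stage mask uses a swirl-distorted sphere that scales with CubeTownCorruptionSpread, eroded by inverted noise.
# Also derives a POI mask from starting-area blue and a texture-size-based one-pixel perimeter mask, noted as unreliable for non-square textures.
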Material /CorruptionGameplay/Materials/M_corruptionGradient_process
graph {
 properties {pixelDepthOffsetMode=PDOM_Legacy, shadingModel=MSM_Unlit, usedShadingModels=Unlit}
 nodes [
  n1 = NamedRerouteDeclaration(None=TextureObjectParameter() {parameterName=Thresholded, texture=&T_BlacksRGB}) {name=ThresholdedRT, nodeColor=<rgba 0 0.9766 1 1>}
  n2 = ScalarParameter() {defaultValue=0.01, parameterName=blurA}
  n3 = TextureSample(UVs=TextureCoordinate() {uTiling=2, vTiling=2}) {samplerType=SAMPLERTYPE_Masks, texture=/Game/Global/Textures/Default/Noise/NoisePattern}
  n4 = MaterialFunctionCall!(
   TextureObject=@n1
   UV
   Distance=ScalarParameter() {defaultValue=0.001, parameterName=blueB}
   Distance Steps=Constant() {r=32}
   Radial Steps=Constant() {r=16}
   Radial Offset
   Kernel Power
  ) {materialFunction=/Engine/Functions/Engine_MaterialFunctions01/Texturing/SpiralBlur-Texture}
  n5 = Add(
   A=TextureSample(Tex=@n1) {samplerType=SAMPLERTYPE_Masks, texture=&NoisePattern}.R
   B=Add(
    A=MaterialFunctionCall(TextureObject=@n1, Distance=@n2) {materialFunction=&SpiralBlur-Texture}.Result
    B=Multiply(A=@n3.G, B=@n4.Result)
   ) {constB=0}
  ) {constB=0}
  n6 = TextureSampleParameter2D!(UVs, Level, Apply View MipBias) {
   constMipValue=0
   mipValueMode=TMVM_MipLevel
   parameterName=StartingCorruption
   samplerType=SAMPLERTYPE_LinearColor
   texture=/Game/Athena/Apollo/Environments/Landscape/Materials/Corruption/T_InitialCorruptionAreas
  }
  Multiply(A=Max(A=Max(A=@n5, B=@n6.R), B=NamedRerouteUsage()), B=NamedRerouteUsage())
  n7 = TextureCoordinate()
  n8 = TextureSample(UVs=Multiply(A=ScalarParameter() {defaultValue=1.8, parameterName=DistortU_TilingScale_A}, B=@n7)) {
   samplerSource=SSM_Wrap_WorldGroupSettings
   samplerType=SAMPLERTYPE_LinearColor
   texture=/Game/Global/Textures/Default/Tileable_Patterns/T_Corruption_SwirlPack
  }
  n9 = TextureSample(UVs=Multiply(A=@n7, B=ScalarParameter() {defaultValue=2.45, parameterName=DistortV_TilingScale_A})) {
   samplerSource=SSM_Wrap_WorldGroupSettings
   samplerType=SAMPLERTYPE_LinearColor
   texture=&T_Corruption_SwirlPack
  }
  n10 = ScalarParameter() {defaultValue=1, parameterName=CubeTownCorruptionSpread}
  n11 = Saturate(None=Multiply(A=@n10) {constB=4})
  n12 = Multiply(
   A=Subtract(A=AppendVector(A=@n8.R, B=@n9.R)) {constB=0.5}
   B=Multiply(A=ScalarParameter() {defaultValue=0.15, parameterName=DistortionA_Strength}, B=@n11)
  )
  n13 = TextureCoordinate()
  n14 = TextureSample(UVs=Multiply(A=ScalarParameter() {defaultValue=0.88, parameterName=DistortU_TilingScale_B}, B=@n13)) {
   samplerSource=SSM_Wrap_WorldGroupSettings
   samplerType=SAMPLERTYPE_LinearColor
   texture=&T_Corruption_SwirlPack
  }
  n15 = TextureSample(UVs=Multiply(A=@n13, B=ScalarParameter() {defaultValue=1.1, parameterName=DistortV_TilingScale_B})) {
   samplerSource=SSM_Wrap_WorldGroupSettings
   samplerType=SAMPLERTYPE_LinearColor
   texture=&T_Corruption_SwirlPack
  }
  n16 = Multiply(
   A=Subtract(A=AppendVector(A=@n14.G, B=@n15.G)) {constB=0.5}
   B=Multiply(A=@n11, B=ScalarParameter() {defaultValue=0.05, parameterName=DistortionB_Strength})
  )
  n17 = SphereMask(
   A=Add(A=TextureCoordinate(), B=Add(A=@n12, B=@n16))
   B=AppendVector(
    A=ScalarParameter() {defaultValue=0.515, parameterName=2ndStageCenter_U}
    B=ScalarParameter() {defaultValue=0.525, parameterName=2ndStageCenter_V}
   )
   Radius=LinearInterpolate(B=ScalarParameter() {defaultValue=0.32, parameterName=2ndStageScale}, Alpha=@n10) {constB=0.2}
   Hardness=ScalarParameter() {defaultValue=0.95, parameterName=2ndStageHardness}
  )
  n18 = TextureCoordinate()
  n19 = TextureSample(UVs=Multiply(A=ScalarParameter() {defaultValue=7.8039, parameterName=NoiseA_TilingScale}, B=@n18)) {
   samplerSource=SSM_Wrap_WorldGroupSettings
   samplerType=SAMPLERTYPE_LinearColor
   texture=&T_Corruption_SwirlPack
  }
  n20 = TextureSample(UVs=Multiply(A=OneMinus(None=@n18), B=ScalarParameter() {defaultValue=8.7906, parameterName=NoiseB_TilingScale})) {
   samplerSource=SSM_Wrap_WorldGroupSettings
   samplerType=SAMPLERTYPE_LinearColor
   texture=&T_Corruption_SwirlPack
  }
  n21 = Add(A=@n19.A, B=@n20.A)
  n22 = Power(Base=OneMinus(None=Saturate(None=@n21)), Exp=ScalarParameter() {defaultValue=3, parameterName=NoiseExponent})
  n23 = ScalarParameter() {defaultValue=1.2, parameterName=NoiseScale}
  n24 = Saturate(None=Add(A=NamedRerouteUsage(), B=Multiply(A=@n22, B=@n23)))
  NamedRerouteDeclaration(None=Saturate(None=Subtract(A=@n17, B=@n24))) {name=2ndStage, nodeColor=<rgba 0.6641 0 1 1>}
  NamedRerouteDeclaration(None=@n6.B) {name=POI_Mask, nodeColor=<rgba 0.2187 1 0 1>}
  n25 = TextureCoordinate()
  n26 = MaterialFunctionCall(Float2=Abs(None=ConstantBiasScale(None=@n25) {bias=-0.5, scale=2})) {materialFunction=&BreakOutFloat2Components}
  n27 = MaterialFunctionCall(Float2=TextureProperty(Tex=NamedRerouteUsage())) {
   desc="wonky, doesn't support non-square textures"
   materialFunction=&BreakOutFloat2Components
  }
  n28 = Divide(B=Max(A=@n27.R, B=@n27.R))
  NamedRerouteDeclaration(None=SphereMask(A=Max(A=@n26.R, B=@n26.R), B=Constant(), Radius=OneMinus(None=@n28), Hardness=Constant() {r=1})) {name=OnePixelPerimeterMask, nodeColor=<rgba 1 0 0.4922 1>}
 ]
}
