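# Material function that scrambles or unscrambles a texture's UVs as "encryption", outputting the sampled red channel and warped UVs.
# An Encrypt static switch sets the direction of a chain of sine-wave offsets, per-tile rotations at two tile sizes and tile-centered scaling.
# It also selects a plain sample or one with biased derivatives from unwarped UVs, with a DebugMips explicit-level option.
# A channel-swap, invert and sRGB conversion branch is left unconnected.
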MaterialFunction /Game/Packages/Fortress_SharedMaterials/Base_Material_Functions/MF_MagicR
graph {
 properties {bExposeToLibrary=true}
 nodes [
  n1 = FunctionInput(Preview=TextureObject()) {bUsePreviewValueAsDefault=true, inputName=Texture, inputType=FunctionInput_Texture2D}
  n2 = ScalarParameter() {defaultValue=0.9, parameterName=ENC_ScaleCenter}
  n3 = StaticSwitch(True=@n2, False=Divide(A=Constant() {r=1}, B=@n2), Value=NamedRerouteUsage())
  n4 = Divide(A=NamedRerouteUsage(), B=ScalarParameter() {defaultValue=4, parameterName=TileEncyption2}) {constB=4}
  n5 = Constant3Vector() {constant=<rgba 0 0 1 0>}
  n6 = Multiply(
   A=ScalarParameter() {defaultValue=0.5, parameterName=ENC_Rotation}
   B=StaticSwitch(True=Constant() {r=1}, False=Constant() {r=-1}, Value=NamedRerouteUsage())
  )
  n7 = Divide(A=NamedRerouteUsage(), B=ScalarParameter() {defaultValue=4, parameterName=TileEncyption2}) {constB=4}
  n8 = FunctionInput(Preview=TextureCoordinate()) {bUsePreviewValueAsDefault=true, inputName=UV, inputType=FunctionInput_Vector2, sortPriority=2}
  n9 = Divide(A=ComponentMask(None=@n8) {g=true}, B=NamedRerouteUsage())
  n10 = AppendVector(A=Multiply(A=Sine(None=@n9), B=NamedRerouteUsage()), B=Constant())
  n11 = StaticSwitch(True=Frac(None=Add(A=@n10, B=@n8)), False=@n8, Value=NamedRerouteUsage())
  n12 = Add(A=Floor(None=Multiply(A=@n11, B=@n7)), B=Constant2Vector() {g=0.5, r=0.5})
  n13 = Constant()
  n14 = RotateAboutAxis(
   NormalizedRotationAxis=Constant3Vector() {constant=<rgba 0 0 1 0>}
   RotationAngle=Multiply(A=NamedRerouteUsage()) {constB=-1}
   PivotPoint=AppendVector(A=Divide(A=@n12, B=@n7), B=@n13)
   Position=AppendVector(A=@n11, B=@n13)
  )
  n15 = StaticSwitch(True=@n11, False=Add(A=ComponentMask(None=@n14) {g=true, r=true}, B=@n11), Value=NamedRerouteUsage())
  n16 = Floor(None=Multiply(A=@n15, B=NamedRerouteUsage()))
  n17 = Constant2Vector() {g=0.5, r=0.5}
  n18 = Constant()
  n19 = AppendVector(A=Divide(A=Add(A=@n16, B=@n17), B=NamedRerouteUsage()), B=@n18)
  n20 = AppendVector(A=@n15, B=@n18)
  n21 = Add(
   A=ComponentMask(None=RotateAboutAxis(NormalizedRotationAxis=@n5, RotationAngle=@n6, PivotPoint=@n19, Position=@n20)) {g=true, r=true}
   B=@n15
  )
  n22 = Add(A=Floor(None=Multiply(A=@n21, B=@n4)), B=Constant2Vector() {g=0.5, r=0.5})
  n23 = Constant()
  n24 = RotateAboutAxis(
   NormalizedRotationAxis=Constant3Vector() {constant=<rgba 0 0 1 0>}
   RotationAngle=NamedRerouteUsage()
   PivotPoint=AppendVector(A=Divide(A=@n22, B=@n4), B=@n23)
   Position=AppendVector(A=@n21, B=@n23)
  )
  n25 = StaticSwitch(True=Add(A=ComponentMask(None=@n24) {g=true, r=true}, B=@n21), False=@n21, Value=NamedRerouteUsage())
  n26 = NamedRerouteUsage()
  n27 = Multiply(A=@n25, B=@n26)
  n28 = Divide(A=Add(A=Floor(None=@n27), B=Constant2Vector() {g=0.5, r=0.5}), B=@n26)
  n29 = Subtract(A=Add(A=Divide(A=@n25, B=@n3), B=@n28), B=Divide(A=@n28, B=@n3) {desc=Scale the center offset by the scaling factor}) {desc=Re offset to maintain center}
  n30 = Divide(A=ComponentMask(None=@n29) {g=true}, B=NamedRerouteUsage())
  n31 = NamedRerouteUsage()
  n32 = AppendVector(A=Multiply(A=Sine(None=@n30), B=Multiply(A=@n31) {constB=-1}), B=Constant())
  n33 = StaticSwitch(True=@n29, False=Frac(None=Add(A=@n32, B=@n29)), Value=NamedRerouteUsage())
  n34 = TextureSample(UVs=@n33, Tex=@n1)
  n35 = TextureCoordinate()
  n36 = DDX(Value=@n35)
  n37 = ScalarParameter() {defaultValue=0.375, parameterName=MipBiasDecal}
  n38 = DDY(Value=@n35)
  n39 = StaticSwitch(
   True=TextureSample!(
    UVs=@n33
    Tex=@n1
    Level=FunctionInput() {bUsePreviewValueAsDefault=true, inputName=MipLevel, inputType=FunctionInput_Scalar, sortPriority=12}
    Apply View MipBias
   ) {mipValueMode=TMVM_MipLevel}.RGBA
   False=TextureSample!(UVs=@n33, Tex=@n1, "DDX(UVs)"=Multiply(A=@n36, B=@n37), "DDY(UVs)"=Multiply(A=@n38, B=@n37), Apply View MipBias) {mipValueMode=TMVM_Derivative, samplerType=SAMPLERTYPE_Masks}.RGBA
   Value=FunctionInput(Preview=StaticBool()) {bUsePreviewValueAsDefault=true, inputName=DebugMips, inputType=FunctionInput_StaticBool, sortPriority=11}
  )
  ComponentMask(None=StaticSwitch(True=@n34.RGBA, False=@n39, Value=NamedRerouteUsage())) {a=true}
  FunctionOutput(None=StaticSwitch(
    True=MaterialFunctionCall(Float3=@n34.RGBA) {materialFunction=&BreakOutFloat3Components}.R
    False=MaterialFunctionCall(Float3=@n39) {materialFunction=&BreakOutFloat3Components}.R
    Value=NamedRerouteUsage()
   ))
  FunctionOutput(None=@n33) {outputName=UVs, sortPriority=3}
  NamedRerouteDeclaration(None=ScalarParameter() {defaultValue=32, parameterName=TilesEncryption}) {name=Tiling, nodeColor=<rgba 0 0.6016 1 1>}
  NamedRerouteDeclaration(None=FunctionInput(Preview=StaticBool()) {bUsePreviewValueAsDefault=true, inputName=Encrypt, inputType=FunctionInput_StaticBool, sortPriority=4}) {name=EncryptBool, nodeColor=<rgba 0 0.0859 1 1>}
  n40 = ScalarParameter() {defaultValue=0.1, desc=0.289429, parameterName=Enc_Wavelength}
  NamedRerouteDeclaration(None=@n40) {name=Wavelength, nodeColor=<rgba 0 0.5781 1 1>}
  n41 = ScalarParameter() {defaultValue=0.025, parameterName=Enc_Amplitude}
  NamedRerouteDeclaration(None=@n41) {name=Amplitude, nodeColor=<rgba 1 0.2812 0 1>}
  n42 = ScalarParameter() {defaultValue=0.25, parameterName=ENC_Rotation2}
  NamedRerouteDeclaration(None=@n42) {name=Enc_Rotation, nodeColor=<rgba 0 0.0859 1 1>}
  Reroute(None=@n41)
  Reroute(None=@n42)
  Reroute(None=@n40)
  StaticSwitch(True=@n34.RGBA, False=@n39, Value=NamedRerouteUsage())
  n43 = MaterialFunctionCall(Float3=@n34.RGBA) {materialFunction=&BreakOutFloat3Components}
  n44 = MaterialFunctionCall(Float3=MaterialFunctionCall(
    X=MaterialFunctionCall!(Linear=@n43.R) {materialFunction=/Engine/Functions/Engine_MaterialFunctions01/ImageAdjustment/LinearTosRGB}.sRGB
    Y=MaterialFunctionCall(Linear=@n43.G) {materialFunction=&LinearTosRGB}.sRGB
    Z=MaterialFunctionCall(Linear=@n43.B) {materialFunction=&LinearTosRGB}.sRGB
   ) {materialFunction=&MakeFloat3}.Result) {materialFunction=&BreakOutFloat3Components}
  n45 = MaterialFunctionCall(Float3=@n39) {materialFunction=&BreakOutFloat3Components}
  n46 = MaterialFunctionCall(Float3=MaterialFunctionCall(X=OneMinus(None=@n45.B), Y=@n45.G, Z=OneMinus(None=@n45.R)) {materialFunction=&MakeFloat3}.Result) {materialFunction=&BreakOutFloat3Components}
  StaticSwitch(
   True=MaterialFunctionCall(X=OneMinus(None=@n44.B), Y=@n44.G, Z=OneMinus(None=@n44.R)) {materialFunction=&MakeFloat3}.Result
   False=MaterialFunctionCall(
    X=MaterialFunctionCall!(sRGB=@n46.R) {materialFunction=/Engine/Functions/Engine_MaterialFunctions01/ImageAdjustment/sRGBToLinear}.Linear
    Y=MaterialFunctionCall(sRGB=@n46.G) {materialFunction=&sRGBToLinear}.Linear
    Z=MaterialFunctionCall(sRGB=@n46.B) {materialFunction=&sRGBToLinear}.Linear
   ) {materialFunction=&MakeFloat3}.Result
   Value=NamedRerouteUsage()
  )
 ]
}
