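# Unlit material that thresholds the corruption gradient texture into a binary mask of first-stage corruption.
# A Decrypt parameter blends the raw red channel with the green channel decoded by MF_MagicR.
# Nested Ifs yield 1 where that value is at least a tiny epsilon and no higher than the Corruption1stStage_0-1 collection parameter, otherwise 0.
# The CorruptionTime parameter is present but unconnected.
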Material /CorruptionGameplay/Materials/M_corruptionGradient_threshold
graph {
 properties {pixelDepthOffsetMode=PDOM_Legacy, shadingModel=MSM_Unlit, usedShadingModels=Unlit}
 nodes [
  Constant()
  n1 = TextureObjectParameter() {
   parameterName=GradientTexture
   samplerType=SAMPLERTYPE_LinearColor
   texture=/Game/Athena/Apollo/Environments/Landscape/Materials/Corruption/T_CorruptionGradient
  }
  n2 = MaterialFunctionCall!(Texture=@n1, UV, Encrypt, DebugMips=StaticBoolParameter() {parameterName=DebugMips}, MipLevel=ScalarParameter() {parameterName=MipLevel}) {materialFunction=/Game/Packages/Fortress_SharedMaterials/Base_Material_Functions/MF_MagicR}
  NamedRerouteDeclaration(None=LinearInterpolate(
    A=TextureSample(Tex=@n1).R
    B=MaterialFunctionCall(Color=@n2.Result) {materialFunction=&SplitComponents}.G
    Alpha=ScalarParameter() {defaultValue=1, parameterName=Decrypt}
   )) {name=Decrypt, nodeColor=<rgba 0.8281 1 0 1>}
  n3 = NamedRerouteUsage()
  n4 = CollectionParameter() {collection=&FortniteMaterialParameters, parameterName=Corruption1stStage_0-1}
  n5 = Constant()
  n6 = Constant() {r=1}
  n7 = If(A=@n3, B=@n4, "A > B"=@n5, "A == B"=@n6, "A < B"=@n6)
  Saturate(None=If(A=@n3, B=Constant() {r=0.001}, "A > B"=@n7, "A == B"=@n7, "A < B"=@n5))
  ScalarParameter() {defaultValue=0.5, parameterName=CorruptionTime}
 ]
}
